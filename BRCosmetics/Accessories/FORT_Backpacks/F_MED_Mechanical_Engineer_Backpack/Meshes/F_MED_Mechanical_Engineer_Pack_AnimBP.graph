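# Animation Blueprint for the Mechanical Engineer backpack, an owl that doubles as a skydiving rig.
# Its output is only a MechOwl idle pose through DefaultSlot; the elaborate skydive chain is not connected to it.
# That chain selects skydive or landing state machines, layers turn, sway, feather and drift additives, then pins bones in world space.
# A glider pose snapshot blends in from the pelvis.
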AnimBlueprint /BRCosmetics/Accessories/FORT_Backpacks/F_MED_Mechanical_Engineer_Backpack/Meshes/F_MED_Mechanical_Engineer_Pack_AnimBP
target_skeleton=/Game/Accessories/FORT_Backpacks/F_MED_Mechanical_Engineer_Backpack/Meshes/F_MED_Mechanical_Engineer_Pack_Skeleton
graphs [
 AnimGraph {nodes [
   n1 = "Transform (Modify) Bone - Bone: Glider_Base"(
    ComponentPose=Local To Component(LocalPose=Use cached pose 'PreJoints'())
    Alpha=1
    Translation=Get Attach_Location() -> (Attach_Location:Vector)
    Rotation=Get Attach_Rotation() -> (Attach_Rotation:Rotator)
    Scale="1,1,1"
   ) {
    rotationMode=BMM_Replace
    rotationSpace=BCS_WorldSpace
    translationMode=BMM_Replace
    translationSpace=BCS_WorldSpace
    alphaBoolBlend {blendOption=Linear}
    boneToModify {boneName=Glider_Base}
   }
   n2 = "Transform (Modify) Bone - Bone: Glider_Base"(
    ComponentPose=@n1
    Alpha=GetCurveValue(CurveName=Closed_New)
    Translation=Get Owl_Offset_Tran() -> (Owl_Offset_Tran:Vector)
    Rotation="0,0,0"
    Scale="1,1,1"
   ) {
    rotationMode=BMM_Additive
    rotationSpace=BCS_WorldSpace
    translationMode=BMM_Additive
    translationSpace=BCS_BoneSpace
    alphaBoolBlend {blendOption=Linear}
    boneToModify {boneName=Glider_Base}
   }
   n3 = "Transform (Modify) Bone - Bone: Body_Offset"(
    ComponentPose=@n2
    Alpha=GetCurveValue(CurveName=Backpack)
    Translation=Get Frame_Location() -> (Frame_Location:Vector)
    Rotation=Get Frame_Rot() -> (Frame_Rot:Rotator)
    Scale="1,1,1"
   ) {
    rotationMode=BMM_Replace
    rotationSpace=BCS_WorldSpace
    translationMode=BMM_Replace
    translationSpace=BCS_WorldSpace
    alphaBoolBlend {blendOption=Linear}
    boneToModify {boneName=Body_Offset}
   }
   n4 = "Transform (Modify) Bone - Bone: Glider_Base"(
    ComponentPose=Local To Component(LocalPose=Use cached pose 'PreJoints'())
    Alpha=Get ClosedAlpha() -> (ClosedAlpha:double)
    Translation=Get Attach_Location()
    Rotation=Get Attach_Rotation()
    Scale="1,1,1"
   ) {
    rotationMode=BMM_Replace
    rotationSpace=BCS_WorldSpace
    translationMode=BMM_Replace
    translationSpace=BCS_WorldSpace
    alphaBoolBlend {blendOption=Linear}
    boneToModify {boneName=Glider_Base}
   }
   n5 = "Transform (Modify) Bone - Bone: Glider_Base"(ComponentPose=@n4, Alpha=Get ClosedAlpha(), Translation=Get Owl_Offset_Tran(), Rotation="0,0,0", Scale="1,1,1") {
    rotationMode=BMM_Additive
    rotationSpace=BCS_WorldSpace
    translationMode=BMM_Additive
    translationSpace=BCS_BoneSpace
    alphaBoolBlend {blendOption=Linear}
    boneToModify {boneName=Glider_Base}
   }
   n6 = "Transform (Modify) Bone - Bone: Frame"!(ComponentPose=@n5, bAlphaBoolEnabled:bool=Get HideOwl() -> (HideOwl:bool), Translation="0,0,0", Rotation="0,0,0", Scale="0.001,0.001,0.001") -> (Pose) {alphaInputType=Bool, scaleMode=BMM_Replace, alphaBoolBlend {blendOption=Linear}, boneToModify {boneName=Frame}, scale=<0.001 0.001 0.001>}
   n7 = "Transform (Modify) Bone - Bone: Frame"!(
    ComponentPose=@n6
    Alpha:float=Get OwlFrame_Alpha() -> (OwlFrame_Alpha:double)
    Translation=Get PlayerRoot_Loc() -> (PlayerRoot_Loc:Vector)
    Rotation=Get PlayerRoot_Rot() -> (PlayerRoot_Rot:Rotator)
    Scale="1,1,1"
   ) -> (Pose) {
    rotationSpace=BCS_WorldSpace
    translationMode=BMM_Replace
    translationSpace=BCS_WorldSpace
    alphaBoolBlend {blendOption=Linear}
    boneToModify {boneName=Frame}
   }
   n8 = "Transform (Modify) Bone - Bone: Frame"(
    ComponentPose=@n7
    Alpha=Get OwlFrame_Alpha()
    Translation=Get GliderOwl_Frame_Loc() -> (GliderOwl_Frame_Loc:Vector)
    Rotation=Get GliderOwl_Frame_Rot() -> (GliderOwl_Frame_Rot:Rotator)
    Scale="1,1,1"
   ) {
    rotationMode=BMM_Replace
    rotationSpace=BCS_WorldSpace
    translationMode=BMM_Additive
    translationSpace=BCS_WorldSpace
    alphaBoolBlend {blendOption=Linear}
    boneToModify {boneName=Frame}
   }
   Layered blend per bone(
    BasePose=Blend Poses by bool(
     BlendPose_0=Component To Local(ComponentPose=@n3)
     BlendPose_1=Component To Local(ComponentPose=@n8)
     BlendTime_0=0
     BlendTime_1=0
     bActiveValue=Get LandingPhase() -> (LandingPhase:bool)
    ) {blendType=Linear}
    BlendPoses_0=Pose Snapshot!(Snapshot:Pose Snapshot Structure=Get Glider_Pose() -> (Glider_Pose:Pose Snapshot Structure)) -> (Pose:Pose Link Structure) {is=title AnimGraphNode_PoseSnapshot, mode=SnapshotPin}
    BlendWeights_0=Get Owl_Pose_Alpha() -> (Owl_Pose_Alpha:double)
   ) {bBlendRootMotionBasedOnRootBone=false, layerSetup [{branchFilters [{boneName=pelvis}]}]}
   Input Pose()
   Output Pose(Result=Slot 'DefaultSlot'(Source=Sequence Player 'Idle_Pose_BACKPACK'() {sequence=/BRCosmetics/Animation/Game/Pets/MechOwl/Backpack/Idle/Idle_Pose_BACKPACK})) {layerGroup=None}
   AnimGraphNode_StateMachine_0 = SkyDiveMachine() {is=title AnimGraphNode_StateMachine}
   AnimGraphNode_StateMachine_1 = LandingMachine() {is=title AnimGraphNode_StateMachine}
   n9 = Get CanPlayLandingAnim() -> (CanPlayLandingAnim:bool)
   n10 = Get Skydive_AimPitch() -> (Skydive_AimPitch:double)
   n11 = Get Skydive_AimYaw() -> (Skydive_AimYaw:double)
   n12 = Get AdditiveAlpha() -> (AdditiveAlpha:double)
   n13 = Apply Additive(
    Base=Apply Additive(
     Base=Blend Poses by bool(BlendPose_0=@AnimGraphNode_StateMachine_1, BlendPose_1=@AnimGraphNode_StateMachine_0, bActiveValue=@n9) {blendType=Linear}
     Additive=Blendspace Player 'BS_MechanicalEngineer_Turn_BACKPACK'(X=@n11, Y=@n10) {blendSpace=/BRCosmetics/Accessories/FORT_Backpacks/F_MED_Mechanical_Engineer_Backpack/Animation/BS_MechanicalEngineer_Turn_BACKPACK}
     Alpha=@n12
    ) {alphaBoolBlend {blendOption=Linear}}
    Additive="Sequence Player (additive) 'MechanicalEngineer_Glide_WingSway_BACKPACK'"() {sequence=/BRCosmetics/Accessories/FORT_Backpacks/F_MED_Mechanical_Engineer_Backpack/Animation/MechanicalEngineer_Glide_WingSway_BACKPACK}
    Alpha=@n12
   ) {alphaBoolBlend {blendOption=Linear}}
   n14 = "Sequence Player (additive) 'MechanicalEngineer_Glide_Feathers_BACKPACK'"() {sequence=/BRCosmetics/Accessories/FORT_Backpacks/F_MED_Mechanical_Engineer_Backpack/Animation/MechanicalEngineer_Glide_Feathers_BACKPACK}
   n15 = Get AdditiveAlpha()
   n16 = Random Sequence Player!() -> (Pose:Pose Link Structure) {
    is=title AnimGraphNode_RandomPlayer
    entries [
     {
      chanceToPlay=1
      maxLoopCount=0
      maxPlayRate=1
      minLoopCount=0
      minPlayRate=1
      sequence=/BRCosmetics/Accessories/FORT_Backpacks/F_MED_Mechanical_Engineer_Backpack/Animation/MechanicalEngineer_Glide_DriftA_BACKPACK
      blendIn {blendOption=Linear, blendTime=0.2, customCurve=null}
     }
     {
      chanceToPlay=1
      maxLoopCount=0
      maxPlayRate=1
      minLoopCount=0
      minPlayRate=1
      sequence=/BRCosmetics/Accessories/FORT_Backpacks/F_MED_Mechanical_Engineer_Backpack/Animation/MechanicalEngineer_Glide_DriftB_BACKPACK
      blendIn {blendOption=Linear, blendTime=0.2, customCurve=null}
     }
     {
      chanceToPlay=1
      maxLoopCount=0
      maxPlayRate=1
      minLoopCount=0
      minPlayRate=1
      sequence=/BRCosmetics/Accessories/FORT_Backpacks/F_MED_Mechanical_Engineer_Backpack/Animation/MechanicalEngineer_Glide_DriftC_BACKPACK
      blendIn {blendOption=Linear, blendTime=0.2, customCurve=null}
     }
    ]
   }
   Save cached pose 'PreJoints'(Pose=Apply Additive(Base=Apply Additive(Base=@n13, Additive=@n14, Alpha=@n15) {alphaBoolBlend {blendOption=Linear}}, Additive=@n16, Alpha=@n15) {alphaBoolBlend {blendOption=Linear}}) {cacheName=PreJoints}
   Save cached pose 'ByPassAdditives'() {cacheName=ByPassAdditives}
   "Sequence Player (additive) 'MechanicalEngineer_Glide_DriftA_BACKPACK'"() {sequence=&MechanicalEngineer_Glide_DriftA_BACKPACK}
   Use cached pose 'ByPassAdditives'()
   Use cached pose 'ByPassAdditives'()
   Get Owl_Offset_Rot() -> (Owl_Offset_Rot:Rotator)
   Get Owl_Offset_Rot()
  ]}
 SkyDiveMachine {
  entry=Closed
  path=AnimGraph.AnimGraphNode_StateMachine_0.SkyDiveMachine
  ids=id
  nodes [
   AnimStateNode_0 = Glide() {is=title AnimStateNode}
   AnimStateNode_1 = Dive() {is=title AnimStateNode}
   AnimStateNode_2 = Passthrough() {is=title AnimStateNode}
   AnimStateNode_3 = to_Dive() {is=title AnimStateNode}
   AnimStateNode_4 = to_Glide() {is=title AnimStateNode}
   AnimStateNode_5 = Detach() {is=title AnimStateNode}
   AnimStateNode_6 = to_Closed() {
    is=title AnimStateNode
    stateEntered {guid=#1}
    stateFullyBlended {guid=#2}
    stateLeft {guid=#3}
   }
   AnimStateNode_7 = to_Skydive() {
    is=title AnimStateNode
    stateEntered {guid=#4}
    stateFullyBlended {guid=#5}
    stateLeft {guid=#6}
   }
   AnimStateNode_8 = Closed() {is=title AnimStateNode}
   AnimStateTransitionNode_0 = AnimStateTransitionNode() {crossfadeDuration=0}
   AnimStateTransitionNode_1 = AnimStateTransitionNode() {crossfadeDuration=0}
   AnimStateTransitionNode_10 = AnimStateTransitionNode() {bAutomaticRuleBasedOnSequencePlayerInState=true}
   AnimStateTransitionNode_11 = AnimStateTransitionNode() {crossfadeDuration=0}
   AnimStateTransitionNode_12 = AnimStateTransitionNode() {crossfadeDuration=0}
   AnimStateTransitionNode_13 = AnimStateTransitionNode() {crossfadeDuration=0}
   AnimStateTransitionNode_14 = AnimStateTransitionNode() {crossfadeDuration=0}
   AnimStateTransitionNode_17 = AnimStateTransitionNode()
   AnimStateTransitionNode_18 = AnimStateTransitionNode() {crossfadeDuration=0}
   AnimStateTransitionNode_19 = AnimStateTransitionNode() {crossfadeDuration=0}
   AnimStateTransitionNode_2 = AnimStateTransitionNode() {crossfadeDuration=0.15}
   AnimStateTransitionNode_20 = AnimStateTransitionNode()
   AnimStateTransitionNode_22 = AnimStateTransitionNode()
   AnimStateTransitionNode_23 = AnimStateTransitionNode()
   AnimStateTransitionNode_3 = AnimStateTransitionNode() {crossfadeDuration=0.15}
   AnimStateTransitionNode_4 = AnimStateTransitionNode() {bAutomaticRuleBasedOnSequencePlayerInState=true, crossfadeDuration=0}
   AnimStateTransitionNode_5 = AnimStateTransitionNode() {bAutomaticRuleBasedOnSequencePlayerInState=true, crossfadeDuration=0}
   AnimStateTransitionNode_6 = AnimStateTransitionNode()
   AnimStateTransitionNode_7 = AnimStateTransitionNode()
   AnimStateTransitionNode_8 = AnimStateTransitionNode() {bAutomaticRuleBasedOnSequencePlayerInState=true, crossfadeDuration=0}
   AnimStateTransitionNode_9 = AnimStateTransitionNode() {crossfadeDuration=0}
  ]
  links [
   {
    dst=to_Skydive
    rule=AnimGraph.AnimGraphNode_StateMachine_0.SkyDiveMachine.AnimStateTransitionNode_9.AnimationTransitionGraph_1
    src=Closed
    via=AnimStateTransitionNode_9
   }
   {
    dst=Dive
    rule=AnimGraph.AnimGraphNode_StateMachine_0.SkyDiveMachine.AnimStateTransitionNode_19.AnimationTransitionGraph_0
    src=Detach
    via=AnimStateTransitionNode_19
   }
   {
    dst=Glide
    rule=AnimGraph.AnimGraphNode_StateMachine_0.SkyDiveMachine.AnimStateTransitionNode_18.AnimationTransitionGraph_0
    src=Detach
    via=AnimStateTransitionNode_18
   }
   {
    dst=Detach
    rule=AnimGraph.AnimGraphNode_StateMachine_0.SkyDiveMachine.AnimStateTransitionNode_20.AnimationTransitionGraph_0
    src=Dive
    via=AnimStateTransitionNode_20
   }
   {
    dst=to_Closed
    rule=AnimGraph.AnimGraphNode_StateMachine_0.SkyDiveMachine.AnimStateTransitionNode_11.Transition
    src=Dive
    via=AnimStateTransitionNode_11
   }
   {
    dst=to_Glide
    rule=AnimGraph.AnimGraphNode_StateMachine_0.SkyDiveMachine.AnimStateTransitionNode_3.AnimationTransitionGraph_0
    src=Dive
    via=AnimStateTransitionNode_3
   }
   {
    dst=Detach
    rule=AnimGraph.AnimGraphNode_StateMachine_0.SkyDiveMachine.AnimStateTransitionNode_17.AnimationTransitionGraph_0
    src=Glide
    via=AnimStateTransitionNode_17
   }
   {
    dst=to_Closed
    rule=AnimGraph.AnimGraphNode_StateMachine_0.SkyDiveMachine.AnimStateTransitionNode_1.AnimationTransitionGraph_1
    src=Glide
    via=AnimStateTransitionNode_1
   }
   {
    dst=to_Dive
    rule=AnimGraph.AnimGraphNode_StateMachine_0.SkyDiveMachine.AnimStateTransitionNode_2.AnimationTransitionGraph_0
    src=Glide
    via=AnimStateTransitionNode_2
   }
   {
    dst=Glide
    rule=AnimGraph.AnimGraphNode_StateMachine_0.SkyDiveMachine.AnimStateTransitionNode_0.AnimationTransitionGraph_0
    src=Passthrough
    via=AnimStateTransitionNode_0
   }
   {
    dst=Closed
    rule=AnimGraph.AnimGraphNode_StateMachine_0.SkyDiveMachine.AnimStateTransitionNode_8.AnimationTransitionGraph_1
    src=to_Closed
    via=AnimStateTransitionNode_8
   }
   {
    dst=Detach
    rule=AnimGraph.AnimGraphNode_StateMachine_0.SkyDiveMachine.AnimStateTransitionNode_22.AnimationTransitionGraph_0
    src=to_Dive
    via=AnimStateTransitionNode_22
   }
   {
    dst=Dive
    rule=AnimGraph.AnimGraphNode_StateMachine_0.SkyDiveMachine.AnimStateTransitionNode_4.AnimationTransitionGraph_0
    src=to_Dive
    via=AnimStateTransitionNode_4
   }
   {
    dst=to_Closed
    rule=AnimGraph.AnimGraphNode_StateMachine_0.SkyDiveMachine.AnimStateTransitionNode_13.Transition
    src=to_Dive
    via=AnimStateTransitionNode_13
   }
   {
    dst=to_Glide
    rule=AnimGraph.AnimGraphNode_StateMachine_0.SkyDiveMachine.AnimStateTransitionNode_7.AnimationTransitionGraph_0
    src=to_Dive
    via=AnimStateTransitionNode_7
   }
   {
    dst=Detach
    rule=AnimGraph.AnimGraphNode_StateMachine_0.SkyDiveMachine.AnimStateTransitionNode_23.AnimationTransitionGraph_0
    src=to_Glide
    via=AnimStateTransitionNode_23
   }
   {
    dst=Glide
    rule=AnimGraph.AnimGraphNode_StateMachine_0.SkyDiveMachine.AnimStateTransitionNode_5.AnimationTransitionGraph_0
    src=to_Glide
    via=AnimStateTransitionNode_5
   }
   {
    dst=to_Closed
    rule=AnimGraph.AnimGraphNode_StateMachine_0.SkyDiveMachine.AnimStateTransitionNode_12.Transition
    src=to_Glide
    via=AnimStateTransitionNode_12
   }
   {
    dst=to_Dive
    rule=AnimGraph.AnimGraphNode_StateMachine_0.SkyDiveMachine.AnimStateTransitionNode_6.AnimationTransitionGraph_0
    src=to_Glide
    via=AnimStateTransitionNode_6
   }
   {
    dst=Glide
    rule=AnimGraph.AnimGraphNode_StateMachine_0.SkyDiveMachine.AnimStateTransitionNode_10.AnimationTransitionGraph_1
    src=to_Skydive
    via=AnimStateTransitionNode_10
   }
   {
    dst=to_Closed
    rule=AnimGraph.AnimGraphNode_StateMachine_0.SkyDiveMachine.AnimStateTransitionNode_14.Transition
    src=to_Skydive
    via=AnimStateTransitionNode_14
   }
  ]
 }
 Glide {
  path=AnimGraph.AnimGraphNode_StateMachine_0.SkyDiveMachine.AnimStateNode_0.Glide
  nodes [Output Animation Pose(Result=Sequence Player 'MechanicalEngineer_Glide_Pose_BACKPACK'() {sequence=/BRCosmetics/Accessories/FORT_Backpacks/F_MED_Mechanical_Engineer_Backpack/Animation/MechanicalEngineer_Glide_Pose_BACKPACK})]
 }
 Dive {
  path=AnimGraph.AnimGraphNode_StateMachine_0.SkyDiveMachine.AnimStateNode_1.Dive
  nodes [Output Animation Pose(Result=Blendspace Player 'BS_MechanicalEngineer_Dive_BACKPACK'(X=Get Skydive_AimYaw(), Y=Get Skydive_AimPitch()) {blendSpace=/BRCosmetics/Accessories/FORT_Backpacks/F_MED_Mechanical_Engineer_Backpack/Animation/BS_MechanicalEngineer_Dive_BACKPACK})]
 }
 Passthrough {
  path=AnimGraph.AnimGraphNode_StateMachine_0.SkyDiveMachine.AnimStateNode_2.Passthrough
  nodes [Output Animation Pose(Result=Local Space Ref Pose())]
 }
 to_Dive {
  path=AnimGraph.AnimGraphNode_StateMachine_0.SkyDiveMachine.AnimStateNode_3.to_Dive
  nodes [
   n1 = Blendspace Player 'BS_MechanicalEngineer_Glide_To_Dive_BACKPACK'(X=Get Skydive_AimYaw(), Y=Get Skydive_AimPitch()) {
    bLoop=false
    blendSpace=/BRCosmetics/Accessories/FORT_Backpacks/F_MED_Mechanical_Engineer_Backpack/Animation/BS_MechanicalEngineer_Glide_To_Dive_BACKPACK
   }
   Output Animation Pose(Result=@n1)
  ]
 }
 to_Glide {
  path=AnimGraph.AnimGraphNode_StateMachine_0.SkyDiveMachine.AnimStateNode_4.to_Glide
  nodes [
   n1 = Blendspace Player 'BS_MechanicalEngineer_Dive_To_Glide_BACKPACK'(X=Get Skydive_AimYaw(), Y=Get Skydive_AimPitch()) {
    bLoop=false
    blendSpace=/BRCosmetics/Accessories/FORT_Backpacks/F_MED_Mechanical_Engineer_Backpack/Animation/BS_MechanicalEngineer_Dive_To_Glide_BACKPACK
   }
   Output Animation Pose(Result=@n1)
  ]
 }
 Detach {
  path=AnimGraph.AnimGraphNode_StateMachine_0.SkyDiveMachine.AnimStateNode_5.Detach
  nodes [
   n1 = Blendspace Player 'BS_MechanicalEngineer_DetachDive_BACKPACK'(X=Get SkydiveForward() -> (SkydiveForward:double), Y=Get Skydive_AimPitch()) {
    bLoop=false
    blendSpace=/BRCosmetics/Accessories/FORT_Backpacks/F_MED_Mechanical_Engineer_Backpack/Animation/Transitions/BS_MechanicalEngineer_DetachDive_BACKPACK
    groupName=detach
    method=SyncGroup
   }
   n2 = Blendspace Player 'BS_MechanicalEngineer_DetachGlide_BACKPACK'(X=Get SkydiveForward(), Y=Get Skydive_AimPitch()) {
    bLoop=false
    blendSpace=/BRCosmetics/Accessories/FORT_Backpacks/F_MED_Mechanical_Engineer_Backpack/Animation/Transitions/BS_MechanicalEngineer_DetachGlide_BACKPACK
    groupName=detach
    method=SyncGroup
   }
   n3 = Get SkydivingState() -> (SkydivingState:Fort Anim Input Skydiving Structure)
   Output Animation Pose(Result=Blend Poses by bool(
     BlendPose_0=@n1
     BlendPose_1=@n2
     bActiveValue=Break Fort Anim Input Skydiving!(FortAnimInput_Skydiving:"Fort Anim Input Skydiving Structure (by ref)"=@n3) -> (
      ParaGilderRootOffset:Vector=~
      ParaGliderRootRotationOverride:Rotator=~
      bUseParaGlideRootModifier:bool=~
      bIsSkydivingFromLaunchPad:bool=~
      bIsSkydivingFromBus:bool=~
      bIsInVortex:bool=~
      bIsDiving:bool=~
      bIsParachuteOpen:bool=~
      bIsSkydiving:bool=~
      bIsParachuteLeaning:bool=~
      bIsSkydiveLeaning:bool=~
      bIsLeaning:bool=~
      bIsSkydiveDiveMode:bool=~
      bPlaySkydiveDrift:bool=~
      bSkydiveDriftDelayActive:bool=~
      bSkydiveDriftAnimAllowed:bool=~
      bIsGliderRight:bool=~
      bIsGliderCenter:bool=~
      bIsGliderLeft:bool=~
      bIsGliderForward:bool=~
      bIsGliderBack:bool=~
      bHasDeployedAndIsStandingUpright:bool=~
      LocalAccelForward:float=~
      LocalAccelRight:float=~
      LocalVelocityRight:float=~
      SkydiveAimPitch:float=~
      SkydiveAimYaw:float=~
      DeployChuteAnimRate:float=~
      SkydiveDriftAnimRate:float=~
      SkydiveDriftAnimRateCurrent:float=~
      SkydiveFidgetAnimRate:float=~
      SkydiveFidgetAnimRateCurrent:float=~
      SkydiveAdditiveAlpha:float=~
      SkydiveDriftDelay:float=~
      SkydiveDriftAnim:int
      SkydiveDriftAnimMax:int
      LaunchpadAnim:int
      LocalAccelDir:ESkydivingDirection Enum
      DirectionLast:ESkydivingDirection Enum
      PoseAdditive:Anim Sequence Object Reference
     ) {is=title K2Node_BreakStruct}.bIsSkydiveDiveMode
    ) {blendType=Linear})
  ]
 }
 to_Closed {
  path=AnimGraph.AnimGraphNode_StateMachine_0.SkyDiveMachine.AnimStateNode_6.to_Closed
  nodes [
   n1 = Sequence Player 'MechanicalEngineer_Glide_to_Umbrella_BACKPACK'() {
    bLoopAnimation=false
    sequence=/Game/Accessories/FORT_Backpacks/F_MED_Mechanical_Engineer_Backpack/Animation/Transitions/MechanicalEngineer_Glide_to_Umbrella_BACKPACK
   }
   n2 = Sequence Player 'MechanicalEngineer_Glide_to_Closed_BACKPACK'() {
    bLoopAnimation=false
    sequence=/Game/Accessories/FORT_Backpacks/F_MED_Mechanical_Engineer_Backpack/Animation/Transitions/MechanicalEngineer_Glide_to_Closed_BACKPACK
   }
   n3 = Get GliderPlayerAnimSet() -> (GliderPlayerAnimSet:Fort Player Glider Anim Set Soft Object Reference)
   n4 = Get UmbrellaAnimSet() -> (UmbrellaAnimSet:Fort Player Glider Anim Set Soft Object Reference)
   n5 = Sequence Player 'MechanicalEngineer_Glide_to_HangGlider_BACKPACK'() {
    bLoopAnimation=false
    sequence=/Game/Accessories/FORT_Backpacks/F_MED_Mechanical_Engineer_Backpack/Animation/Transitions/MechanicalEngineer_Glide_to_HangGlider_BACKPACK
   }
   Blend Poses by bool(
    BlendPose_0=@n5
    BlendPose_1=Blend Poses by bool(
     BlendPose_0=@n1
     BlendPose_1=@n2
     bActiveValue="Equal (SoftObjectReference)"!(A:"Soft Object Reference (by ref)"=@n3, B:"Soft Object Reference (by ref)"=@n4) -> (ReturnValue:bool)
    ) {blendType=Linear}
    bActiveValue="Equal (SoftObjectReference)"(A=Get GliderPlayerAnimSet(), B=Get HangGliderAnimSet() -> (HangGliderAnimSet:Fort Player Glider Anim Set Soft Object Reference))
   ) {blendType=Linear}
   Sequence Player 'MechanicalEngineer_Glide_to_Closed_BACKPACK'() {
    bLoopAnimation=false
    sequence=&MechanicalEngineer_Glide_to_Closed_BACKPACK
   }
   n6 = Sequence Player 'MechanicalEngineer_Open_to_Backpack_BACKPACK'() {
    bLoopAnimation=false
    sequence=/BRCosmetics/Accessories/FORT_Backpacks/F_MED_Mechanical_Engineer_Backpack/Animation/Transitions/MechanicalEngineer_Open_to_Backpack_BACKPACK
   }
   Output Animation Pose(Result=@n6)
  ]
 }
 to_Skydive {
  path=AnimGraph.AnimGraphNode_StateMachine_0.SkyDiveMachine.AnimStateNode_7.to_Skydive
  nodes [
   n1 = Sequence Player 'MechanicalEngineer_Closed_to_Glide_BACKPACK'() {
    bLoopAnimation=false
    sequence=/BRCosmetics/Accessories/FORT_Backpacks/F_MED_Mechanical_Engineer_Backpack/Animation/Transitions/MechanicalEngineer_Closed_to_Glide_BACKPACK
   }
   Output Animation Pose(Result=@n1)
  ]
 }
 Closed {
  path=AnimGraph.AnimGraphNode_StateMachine_0.SkyDiveMachine.AnimStateNode_8.Closed
  nodes [Output Animation Pose(Result=Sequence Player 'MechanicalEngineer_Closed_BACKPACK'() {sequence=/BRCosmetics/Accessories/FORT_Backpacks/F_MED_Mechanical_Engineer_Backpack/Animation/MechanicalEngineer_Closed_BACKPACK})]
 }
 AnimationTransitionGraph_0 {
  path=AnimGraph.AnimGraphNode_StateMachine_0.SkyDiveMachine.AnimStateTransitionNode_0.AnimationTransitionGraph_0
  nodes [Result(bCanEnterTransition=Break Fort Anim Input Skydiving(FortAnimInput_Skydiving=Get SkydivingState()) -> (
     ParaGilderRootOffset=~
     ParaGliderRootRotationOverride=~
     bUseParaGlideRootModifier=~
     bIsSkydivingFromLaunchPad=~
     bIsSkydivingFromBus=~
     bIsInVortex=~
     bIsDiving=~
     bIsParachuteOpen=~
     bIsSkydiving=~
     bIsParachuteLeaning=~
     bIsSkydiveLeaning=~
     bIsLeaning=~
     bIsSkydiveDiveMode=~
     bPlaySkydiveDrift=~
     bSkydiveDriftDelayActive=~
     bSkydiveDriftAnimAllowed=~
     bIsGliderRight=~
     bIsGliderCenter=~
     bIsGliderLeft=~
     bIsGliderForward=~
     bIsGliderBack=~
     bHasDeployedAndIsStandingUpright=~
     LocalAccelForward=~
     LocalAccelRight=~
     LocalVelocityRight=~
     SkydiveAimPitch=~
     SkydiveAimYaw=~
     DeployChuteAnimRate=~
     SkydiveDriftAnimRate=~
     SkydiveDriftAnimRateCurrent=~
     SkydiveFidgetAnimRate=~
     SkydiveFidgetAnimRateCurrent=~
     SkydiveAdditiveAlpha=~
     SkydiveDriftDelay=~
    ).bIsSkydiving)]
 }
 AnimationTransitionGraph_1 {
  path=AnimGraph.AnimGraphNode_StateMachine_0.SkyDiveMachine.AnimStateTransitionNode_1.AnimationTransitionGraph_1
  nodes [
   n1 = Get SkydivingState()
   n2 = Get GliderMatch() -> (GliderMatch:bool)
   Result(bCanEnterTransition=AND Boolean(
     A=Break Fort Anim Input Skydiving(FortAnimInput_Skydiving=@n1) -> (
      ParaGilderRootOffset=~
      ParaGliderRootRotationOverride=~
      bUseParaGlideRootModifier=~
      bIsSkydivingFromLaunchPad=~
      bIsSkydivingFromBus=~
      bIsInVortex=~
      bIsDiving=~
      bIsParachuteOpen=~
      bIsSkydiving=~
      bIsParachuteLeaning=~
      bIsSkydiveLeaning=~
      bIsLeaning=~
      bIsSkydiveDiveMode=~
      bPlaySkydiveDrift=~
      bSkydiveDriftDelayActive=~
      bSkydiveDriftAnimAllowed=~
      bIsGliderRight=~
      bIsGliderCenter=~
      bIsGliderLeft=~
      bIsGliderForward=~
      bIsGliderBack=~
      bHasDeployedAndIsStandingUpright=~
      LocalAccelForward=~
      LocalAccelRight=~
      LocalVelocityRight=~
      SkydiveAimPitch=~
      SkydiveAimYaw=~
      DeployChuteAnimRate=~
      SkydiveDriftAnimRate=~
      SkydiveDriftAnimRateCurrent=~
      SkydiveFidgetAnimRate=~
      SkydiveFidgetAnimRateCurrent=~
      SkydiveAdditiveAlpha=~
      SkydiveDriftDelay=~
     ).bIsParachuteOpen
     B=NOT Boolean(A=@n2)
    ))
   Get BlockAnimRule() -> (BlockAnimRule:bool)
  ]
 }
 AnimationTransitionGraph_1 {
  path=AnimGraph.AnimGraphNode_StateMachine_0.SkyDiveMachine.AnimStateTransitionNode_10.AnimationTransitionGraph_1
  nodes [
   Result()
   Get BlockAnimRule()
  ]
 }
 Transition {
  path=AnimGraph.AnimGraphNode_StateMachine_0.SkyDiveMachine.AnimStateTransitionNode_11.Transition
  nodes [
   n1 = Get SkydivingState()
   n2 = Get GliderMatch()
   Result(bCanEnterTransition=AND Boolean(
     A=Break Fort Anim Input Skydiving(FortAnimInput_Skydiving=@n1) -> (
      ParaGilderRootOffset=~
      ParaGliderRootRotationOverride=~
      bUseParaGlideRootModifier=~
      bIsSkydivingFromLaunchPad=~
      bIsSkydivingFromBus=~
      bIsInVortex=~
      bIsDiving=~
      bIsParachuteOpen=~
      bIsSkydiving=~
      bIsParachuteLeaning=~
      bIsSkydiveLeaning=~
      bIsLeaning=~
      bIsSkydiveDiveMode=~
      bPlaySkydiveDrift=~
      bSkydiveDriftDelayActive=~
      bSkydiveDriftAnimAllowed=~
      bIsGliderRight=~
      bIsGliderCenter=~
      bIsGliderLeft=~
      bIsGliderForward=~
      bIsGliderBack=~
      bHasDeployedAndIsStandingUpright=~
      LocalAccelForward=~
      LocalAccelRight=~
      LocalVelocityRight=~
      SkydiveAimPitch=~
      SkydiveAimYaw=~
      DeployChuteAnimRate=~
      SkydiveDriftAnimRate=~
      SkydiveDriftAnimRateCurrent=~
      SkydiveFidgetAnimRate=~
      SkydiveFidgetAnimRateCurrent=~
      SkydiveAdditiveAlpha=~
      SkydiveDriftDelay=~
     ).bIsParachuteOpen
     B=NOT Boolean(A=@n2)
    ))
  ]
 }
 Transition {
  path=AnimGraph.AnimGraphNode_StateMachine_0.SkyDiveMachine.AnimStateTransitionNode_12.Transition
  nodes [
   n1 = Get SkydivingState()
   n2 = Get GliderMatch()
   Result(bCanEnterTransition=AND Boolean(
     A=Break Fort Anim Input Skydiving(FortAnimInput_Skydiving=@n1) -> (
      ParaGilderRootOffset=~
      ParaGliderRootRotationOverride=~
      bUseParaGlideRootModifier=~
      bIsSkydivingFromLaunchPad=~
      bIsSkydivingFromBus=~
      bIsInVortex=~
      bIsDiving=~
      bIsParachuteOpen=~
      bIsSkydiving=~
      bIsParachuteLeaning=~
      bIsSkydiveLeaning=~
      bIsLeaning=~
      bIsSkydiveDiveMode=~
      bPlaySkydiveDrift=~
      bSkydiveDriftDelayActive=~
      bSkydiveDriftAnimAllowed=~
      bIsGliderRight=~
      bIsGliderCenter=~
      bIsGliderLeft=~
      bIsGliderForward=~
      bIsGliderBack=~
      bHasDeployedAndIsStandingUpright=~
      LocalAccelForward=~
      LocalAccelRight=~
      LocalVelocityRight=~
      SkydiveAimPitch=~
      SkydiveAimYaw=~
      DeployChuteAnimRate=~
      SkydiveDriftAnimRate=~
      SkydiveDriftAnimRateCurrent=~
      SkydiveFidgetAnimRate=~
      SkydiveFidgetAnimRateCurrent=~
      SkydiveAdditiveAlpha=~
      SkydiveDriftDelay=~
     ).bIsParachuteOpen
     B=NOT Boolean(A=@n2)
    ))
  ]
 }
 Transition {
  path=AnimGraph.AnimGraphNode_StateMachine_0.SkyDiveMachine.AnimStateTransitionNode_13.Transition
  nodes [
   n1 = Get SkydivingState()
   n2 = Get GliderMatch()
   Result(bCanEnterTransition=AND Boolean(
     A=Break Fort Anim Input Skydiving(FortAnimInput_Skydiving=@n1) -> (
      ParaGilderRootOffset=~
      ParaGliderRootRotationOverride=~
      bUseParaGlideRootModifier=~
      bIsSkydivingFromLaunchPad=~
      bIsSkydivingFromBus=~
      bIsInVortex=~
      bIsDiving=~
      bIsParachuteOpen=~
      bIsSkydiving=~
      bIsParachuteLeaning=~
      bIsSkydiveLeaning=~
      bIsLeaning=~
      bIsSkydiveDiveMode=~
      bPlaySkydiveDrift=~
      bSkydiveDriftDelayActive=~
      bSkydiveDriftAnimAllowed=~
      bIsGliderRight=~
      bIsGliderCenter=~
      bIsGliderLeft=~
      bIsGliderForward=~
      bIsGliderBack=~
      bHasDeployedAndIsStandingUpright=~
      LocalAccelForward=~
      LocalAccelRight=~
      LocalVelocityRight=~
      SkydiveAimPitch=~
      SkydiveAimYaw=~
      DeployChuteAnimRate=~
      SkydiveDriftAnimRate=~
      SkydiveDriftAnimRateCurrent=~
      SkydiveFidgetAnimRate=~
      SkydiveFidgetAnimRateCurrent=~
      SkydiveAdditiveAlpha=~
      SkydiveDriftDelay=~
     ).bIsParachuteOpen
     B=NOT Boolean(A=@n2)
    ))
  ]
 }
 Transition {
  path=AnimGraph.AnimGraphNode_StateMachine_0.SkyDiveMachine.AnimStateTransitionNode_14.Transition
  nodes [
   n1 = Get SkydivingState()
   n2 = Get GliderMatch()
   Result(bCanEnterTransition=AND Boolean(
     A=Break Fort Anim Input Skydiving(FortAnimInput_Skydiving=@n1) -> (
      ParaGilderRootOffset=~
      ParaGliderRootRotationOverride=~
      bUseParaGlideRootModifier=~
      bIsSkydivingFromLaunchPad=~
      bIsSkydivingFromBus=~
      bIsInVortex=~
      bIsDiving=~
      bIsParachuteOpen=~
      bIsSkydiving=~
      bIsParachuteLeaning=~
      bIsSkydiveLeaning=~
      bIsLeaning=~
      bIsSkydiveDiveMode=~
      bPlaySkydiveDrift=~
      bSkydiveDriftDelayActive=~
      bSkydiveDriftAnimAllowed=~
      bIsGliderRight=~
      bIsGliderCenter=~
      bIsGliderLeft=~
      bIsGliderForward=~
      bIsGliderBack=~
      bHasDeployedAndIsStandingUpright=~
      LocalAccelForward=~
      LocalAccelRight=~
      LocalVelocityRight=~
      SkydiveAimPitch=~
      SkydiveAimYaw=~
      DeployChuteAnimRate=~
      SkydiveDriftAnimRate=~
      SkydiveDriftAnimRateCurrent=~
      SkydiveFidgetAnimRate=~
      SkydiveFidgetAnimRateCurrent=~
      SkydiveAdditiveAlpha=~
      SkydiveDriftDelay=~
     ).bIsParachuteOpen
     B=NOT Boolean(A=@n2)
    ))
  ]
 }
 AnimationTransitionGraph_0 {
  path=AnimGraph.AnimGraphNode_StateMachine_0.SkyDiveMachine.AnimStateTransitionNode_17.AnimationTransitionGraph_0
  nodes [Result(bCanEnterTransition=Get DetachFromWrenchTrans() -> (DetachFromWrenchTrans:bool))]
 }
 AnimationTransitionGraph_0 {
  path=AnimGraph.AnimGraphNode_StateMachine_0.SkyDiveMachine.AnimStateTransitionNode_18.AnimationTransitionGraph_0
  nodes [
   n1 = Get DetachFromWrenchTrans()
   n2 = Break Fort Anim Input Skydiving(FortAnimInput_Skydiving=Get SkydivingState()) -> (
    ParaGilderRootOffset=~
    ParaGliderRootRotationOverride=~
    bUseParaGlideRootModifier=~
    bIsSkydivingFromLaunchPad=~
    bIsSkydivingFromBus=~
    bIsInVortex=~
    bIsDiving=~
    bIsParachuteOpen=~
    bIsSkydiving=~
    bIsParachuteLeaning=~
    bIsSkydiveLeaning=~
    bIsLeaning=~
    bIsSkydiveDiveMode=~
    bPlaySkydiveDrift=~
    bSkydiveDriftDelayActive=~
    bSkydiveDriftAnimAllowed=~
    bIsGliderRight=~
    bIsGliderCenter=~
    bIsGliderLeft=~
    bIsGliderForward=~
    bIsGliderBack=~
    bHasDeployedAndIsStandingUpright=~
    LocalAccelForward=~
    LocalAccelRight=~
    LocalVelocityRight=~
    SkydiveAimPitch=~
    SkydiveAimYaw=~
    DeployChuteAnimRate=~
    SkydiveDriftAnimRate=~
    SkydiveDriftAnimRateCurrent=~
    SkydiveFidgetAnimRate=~
    SkydiveFidgetAnimRateCurrent=~
    SkydiveAdditiveAlpha=~
    SkydiveDriftDelay=~
   )
   Result(bCanEnterTransition=AND Boolean(A=NOT Boolean(A=@n1), B=NOT Boolean(A=@n2.bIsSkydiveDiveMode)))
  ]
 }
 AnimationTransitionGraph_0 {
  path=AnimGraph.AnimGraphNode_StateMachine_0.SkyDiveMachine.AnimStateTransitionNode_19.AnimationTransitionGraph_0
  nodes [
   n1 = Get SkydivingState()
   n2 = Get DetachFromWrenchTrans()
   Result(bCanEnterTransition=AND Boolean(
     A=NOT Boolean(A=@n2)
     B=Break Fort Anim Input Skydiving(FortAnimInput_Skydiving=@n1) -> (
      ParaGilderRootOffset=~
      ParaGliderRootRotationOverride=~
      bUseParaGlideRootModifier=~
      bIsSkydivingFromLaunchPad=~
      bIsSkydivingFromBus=~
      bIsInVortex=~
      bIsDiving=~
      bIsParachuteOpen=~
      bIsSkydiving=~
      bIsParachuteLeaning=~
      bIsSkydiveLeaning=~
      bIsLeaning=~
      bIsSkydiveDiveMode=~
      bPlaySkydiveDrift=~
      bSkydiveDriftDelayActive=~
      bSkydiveDriftAnimAllowed=~
      bIsGliderRight=~
      bIsGliderCenter=~
      bIsGliderLeft=~
      bIsGliderForward=~
      bIsGliderBack=~
      bHasDeployedAndIsStandingUpright=~
      LocalAccelForward=~
      LocalAccelRight=~
      LocalVelocityRight=~
      SkydiveAimPitch=~
      SkydiveAimYaw=~
      DeployChuteAnimRate=~
      SkydiveDriftAnimRate=~
      SkydiveDriftAnimRateCurrent=~
      SkydiveFidgetAnimRate=~
      SkydiveFidgetAnimRateCurrent=~
      SkydiveAdditiveAlpha=~
      SkydiveDriftDelay=~
     ).bIsSkydiveDiveMode
    ))
  ]
 }
 AnimationTransitionGraph_0 {
  path=AnimGraph.AnimGraphNode_StateMachine_0.SkyDiveMachine.AnimStateTransitionNode_2.AnimationTransitionGraph_0
  nodes [Result(bCanEnterTransition=Break Fort Anim Input Skydiving(FortAnimInput_Skydiving=Get SkydivingState()) -> (
     ParaGilderRootOffset=~
     ParaGliderRootRotationOverride=~
     bUseParaGlideRootModifier=~
     bIsSkydivingFromLaunchPad=~
     bIsSkydivingFromBus=~
     bIsInVortex=~
     bIsDiving=~
     bIsParachuteOpen=~
     bIsSkydiving=~
     bIsParachuteLeaning=~
     bIsSkydiveLeaning=~
     bIsLeaning=~
     bIsSkydiveDiveMode=~
     bPlaySkydiveDrift=~
     bSkydiveDriftDelayActive=~
     bSkydiveDriftAnimAllowed=~
     bIsGliderRight=~
     bIsGliderCenter=~
     bIsGliderLeft=~
     bIsGliderForward=~
     bIsGliderBack=~
     bHasDeployedAndIsStandingUpright=~
     LocalAccelForward=~
     LocalAccelRight=~
     LocalVelocityRight=~
     SkydiveAimPitch=~
     SkydiveAimYaw=~
     DeployChuteAnimRate=~
     SkydiveDriftAnimRate=~
     SkydiveDriftAnimRateCurrent=~
     SkydiveFidgetAnimRate=~
     SkydiveFidgetAnimRateCurrent=~
     SkydiveAdditiveAlpha=~
     SkydiveDriftDelay=~
    ).bIsSkydiveDiveMode)]
 }
 AnimationTransitionGraph_0 {
  path=AnimGraph.AnimGraphNode_StateMachine_0.SkyDiveMachine.AnimStateTransitionNode_20.AnimationTransitionGraph_0
  nodes [Result(bCanEnterTransition=Get DetachFromWrenchTrans())]
 }
 AnimationTransitionGraph_0 {
  path=AnimGraph.AnimGraphNode_StateMachine_0.SkyDiveMachine.AnimStateTransitionNode_22.AnimationTransitionGraph_0
  nodes [Result(bCanEnterTransition=Get DetachFromWrenchTrans())]
 }
 AnimationTransitionGraph_0 {
  path=AnimGraph.AnimGraphNode_StateMachine_0.SkyDiveMachine.AnimStateTransitionNode_23.AnimationTransitionGraph_0
  nodes [Result(bCanEnterTransition=Get DetachFromWrenchTrans())]
 }
 AnimationTransitionGraph_0 {
  path=AnimGraph.AnimGraphNode_StateMachine_0.SkyDiveMachine.AnimStateTransitionNode_3.AnimationTransitionGraph_0
  nodes [
   n1 = Get SkydivingState()
   Result(bCanEnterTransition=NOT Boolean(A=Break Fort Anim Input Skydiving(FortAnimInput_Skydiving=@n1) -> (
      ParaGilderRootOffset=~
      ParaGliderRootRotationOverride=~
      bUseParaGlideRootModifier=~
      bIsSkydivingFromLaunchPad=~
      bIsSkydivingFromBus=~
      bIsInVortex=~
      bIsDiving=~
      bIsParachuteOpen=~
      bIsSkydiving=~
      bIsParachuteLeaning=~
      bIsSkydiveLeaning=~
      bIsLeaning=~
      bIsSkydiveDiveMode=~
      bPlaySkydiveDrift=~
      bSkydiveDriftDelayActive=~
      bSkydiveDriftAnimAllowed=~
      bIsGliderRight=~
      bIsGliderCenter=~
      bIsGliderLeft=~
      bIsGliderForward=~
      bIsGliderBack=~
      bHasDeployedAndIsStandingUpright=~
      LocalAccelForward=~
      LocalAccelRight=~
      LocalVelocityRight=~
      SkydiveAimPitch=~
      SkydiveAimYaw=~
      DeployChuteAnimRate=~
      SkydiveDriftAnimRate=~
      SkydiveDriftAnimRateCurrent=~
      SkydiveFidgetAnimRate=~
      SkydiveFidgetAnimRateCurrent=~
      SkydiveAdditiveAlpha=~
      SkydiveDriftDelay=~
     ).bIsSkydiveDiveMode))
  ]
 }
 AnimationTransitionGraph_0 {path=AnimGraph.AnimGraphNode_StateMachine_0.SkyDiveMachine.AnimStateTransitionNode_4.AnimationTransitionGraph_0, nodes [Result()]}
 AnimationTransitionGraph_0 {path=AnimGraph.AnimGraphNode_StateMachine_0.SkyDiveMachine.AnimStateTransitionNode_5.AnimationTransitionGraph_0, nodes [Result()]}
 AnimationTransitionGraph_0 {
  path=AnimGraph.AnimGraphNode_StateMachine_0.SkyDiveMachine.AnimStateTransitionNode_6.AnimationTransitionGraph_0
  nodes [Result(bCanEnterTransition=Break Fort Anim Input Skydiving(FortAnimInput_Skydiving=Get SkydivingState()) -> (
     ParaGilderRootOffset=~
     ParaGliderRootRotationOverride=~
     bUseParaGlideRootModifier=~
     bIsSkydivingFromLaunchPad=~
     bIsSkydivingFromBus=~
     bIsInVortex=~
     bIsDiving=~
     bIsParachuteOpen=~
     bIsSkydiving=~
     bIsParachuteLeaning=~
     bIsSkydiveLeaning=~
     bIsLeaning=~
     bIsSkydiveDiveMode=~
     bPlaySkydiveDrift=~
     bSkydiveDriftDelayActive=~
     bSkydiveDriftAnimAllowed=~
     bIsGliderRight=~
     bIsGliderCenter=~
     bIsGliderLeft=~
     bIsGliderForward=~
     bIsGliderBack=~
     bHasDeployedAndIsStandingUpright=~
     LocalAccelForward=~
     LocalAccelRight=~
     LocalVelocityRight=~
     SkydiveAimPitch=~
     SkydiveAimYaw=~
     DeployChuteAnimRate=~
     SkydiveDriftAnimRate=~
     SkydiveDriftAnimRateCurrent=~
     SkydiveFidgetAnimRate=~
     SkydiveFidgetAnimRateCurrent=~
     SkydiveAdditiveAlpha=~
     SkydiveDriftDelay=~
    ).bIsSkydiveDiveMode)]
 }
 AnimationTransitionGraph_0 {
  path=AnimGraph.AnimGraphNode_StateMachine_0.SkyDiveMachine.AnimStateTransitionNode_7.AnimationTransitionGraph_0
  nodes [
   n1 = Get SkydivingState()
   Result(bCanEnterTransition=NOT Boolean(A=Break Fort Anim Input Skydiving(FortAnimInput_Skydiving=@n1) -> (
      ParaGilderRootOffset=~
      ParaGliderRootRotationOverride=~
      bUseParaGlideRootModifier=~
      bIsSkydivingFromLaunchPad=~
      bIsSkydivingFromBus=~
      bIsInVortex=~
      bIsDiving=~
      bIsParachuteOpen=~
      bIsSkydiving=~
      bIsParachuteLeaning=~
      bIsSkydiveLeaning=~
      bIsLeaning=~
      bIsSkydiveDiveMode=~
      bPlaySkydiveDrift=~
      bSkydiveDriftDelayActive=~
      bSkydiveDriftAnimAllowed=~
      bIsGliderRight=~
      bIsGliderCenter=~
      bIsGliderLeft=~
      bIsGliderForward=~
      bIsGliderBack=~
      bHasDeployedAndIsStandingUpright=~
      LocalAccelForward=~
      LocalAccelRight=~
      LocalVelocityRight=~
      SkydiveAimPitch=~
      SkydiveAimYaw=~
      DeployChuteAnimRate=~
      SkydiveDriftAnimRate=~
      SkydiveDriftAnimRateCurrent=~
      SkydiveFidgetAnimRate=~
      SkydiveFidgetAnimRateCurrent=~
      SkydiveAdditiveAlpha=~
      SkydiveDriftDelay=~
     ).bIsSkydiveDiveMode))
  ]
 }
 AnimationTransitionGraph_1 {path=AnimGraph.AnimGraphNode_StateMachine_0.SkyDiveMachine.AnimStateTransitionNode_8.AnimationTransitionGraph_1, nodes [Result()]}
 AnimationTransitionGraph_1 {
  path=AnimGraph.AnimGraphNode_StateMachine_0.SkyDiveMachine.AnimStateTransitionNode_9.AnimationTransitionGraph_1
  nodes [
   n1 = Break Fort Anim Input Skydiving(FortAnimInput_Skydiving=Get SkydivingState()) -> (
    ParaGilderRootOffset=~
    ParaGliderRootRotationOverride=~
    bUseParaGlideRootModifier=~
    bIsSkydivingFromLaunchPad=~
    bIsSkydivingFromBus=~
    bIsInVortex=~
    bIsDiving=~
    bIsParachuteOpen=~
    bIsSkydiving=~
    bIsParachuteLeaning=~
    bIsSkydiveLeaning=~
    bIsLeaning=~
    bIsSkydiveDiveMode=~
    bPlaySkydiveDrift=~
    bSkydiveDriftDelayActive=~
    bSkydiveDriftAnimAllowed=~
    bIsGliderRight=~
    bIsGliderCenter=~
    bIsGliderLeft=~
    bIsGliderForward=~
    bIsGliderBack=~
    bHasDeployedAndIsStandingUpright=~
    LocalAccelForward=~
    LocalAccelRight=~
    LocalVelocityRight=~
    SkydiveAimPitch=~
    SkydiveAimYaw=~
    DeployChuteAnimRate=~
    SkydiveDriftAnimRate=~
    SkydiveDriftAnimRateCurrent=~
    SkydiveFidgetAnimRate=~
    SkydiveFidgetAnimRateCurrent=~
    SkydiveAdditiveAlpha=~
    SkydiveDriftDelay=~
   )
   Result(bCanEnterTransition=AND Boolean(A=NOT Boolean(A=@n1.bIsParachuteOpen), B=@n1.bIsSkydiving))
  ]
 }
 LandingMachine {
  entry=to_Closed
  path=AnimGraph.AnimGraphNode_StateMachine_1.LandingMachine
  ids=id
  nodes [
   AnimStateNode_0 = to_Closed()
   AnimStateNode_1 = Closed()
   AnimStateTransitionNode_2 = AnimStateTransitionNode() {bAutomaticRuleBasedOnSequencePlayerInState=true, crossfadeDuration=0}
  ]
  links [{
    dst=Closed
    rule=AnimGraph.AnimGraphNode_StateMachine_1.LandingMachine.AnimStateTransitionNode_2.Transition
    src=to_Closed
    via=AnimStateTransitionNode_2
   }]
 }
 to_Closed {
  path=AnimGraph.AnimGraphNode_StateMachine_1.LandingMachine.AnimStateNode_0.to_Closed
  nodes [
   n1 = Sequence Player 'MechanicalEngineer_Open_to_Backpack_BACKPACK'() {
    bLoopAnimation=false
    sequence=&MechanicalEngineer_Open_to_Backpack_BACKPACK
   }
   Output Animation Pose(Result=@n1)
  ]
 }
 Closed {
  path=AnimGraph.AnimGraphNode_StateMachine_1.LandingMachine.AnimStateNode_1.Closed
  nodes [Output Animation Pose(Result=Sequence Player 'MechanicalEngineer_Closed_BACKPACK'() {sequence=&MechanicalEngineer_Closed_BACKPACK})]
 }
 Transition {path=AnimGraph.AnimGraphNode_StateMachine_1.LandingMachine.AnimStateTransitionNode_2.Transition, nodes [Result()]}
 CheckGlider {nodes [
   Inputs()
   > Do Once()
   > n1 = Cast To FortPlayerPawn(Object=TryGetPawnOwner())
   > Set PlayerMesh(PlayerMesh:Skeletal Mesh Component Object Reference=Get Mesh(self=@n1)) -> (Output_Get:Skeletal Mesh Component Object Reference)
   > Set CharMovement(CharMovement:Character Movement Component Object Reference=Get CharacterMovement!(self:Character Object Reference=@n1) -> (CharacterMovement:Character Movement Component Object Reference)) -> (Output_Get:Character Movement Component Object Reference)
   > Set BackpackMesh(BackpackMesh:Skeletal Mesh Component Object Reference=GetSkeletalMeshForPartType(self=@n1, PartType=Backpack)) -> (Output_Get:Skeletal Mesh Component Object Reference)
   n2 = GetItemDefinitionTypeFromLoadoutWithActor!(
    Actor:Actor Object Reference=@n1
    LoadoutSlot:Gameplay Tag Structure=(TagName="LoadoutSlot.Glider")
    CosmeticType:Athena Cosmetic Item Definition Class Reference=/Script/FortniteGame.AthenaGliderItemDefinition
   ) -> (OutCosmeticItemDef:Athena Glider Item Definition Object Reference)
   > Set GliderMatch(GliderMatch:bool="Equal (SoftObjectReference)"(
     A=Get GliderSkelMesh() -> (GliderSkelMesh:Skeletal Mesh Soft Object Reference)
     B=GetOverrideSkelMeshPtr!(self:Athena Glider Item Definition Object Reference=@n2) -> (ReturnValue:Skeletal Mesh Soft Object Reference)
    )) -> (Output_Get:bool)
   > n3 = Set GliderPlayerAnimSet(GliderPlayerAnimSet:Fort Player Glider Anim Set Soft Object Reference=GetPlayerAnimSetPtr!(self:Athena Glider Item Definition Object Reference=@n2) -> (ReturnValue:Fort Player Glider Anim Set Soft Object Reference)) -> (Output_Get:Fort Player Glider Anim Set Soft Object Reference)
   Outputs!(then:Exec=[@n3.then, @n4.then])
   n4 = Set CanPlayLandingAnim(CanPlayLandingAnim:bool=Get GliderMatch()) -> (Output_Get:bool)
   Contains(
    SearchIn="To String (Text)"(InText=GetGliderDisplayName!(self:Fort Player Pawn Object Reference) -> (ReturnValue:Text))
    Substring=MechanicalEngineerFemale
    bSearchFromEnd=true
   )
  ]}
 CheckHideOwlStatus {nodes [
   Inputs()
   > n1 = Cast To FortPlayerPawn(Object=TryGetPawnOwner())
   n2 = Get ParachuteAttachment!(self:Fort Player Pawn Object Reference=@n1) -> (ParachuteAttachment:Fort Player Parachute Object Reference)
   > Is Valid(InputObject=@n2)
   n3 = Get ParachuteMesh!(self:Fort Player Parachute Object Reference=@n2) -> (ParachuteMesh:Skeletal Mesh Component Budgeted Object Reference)
   > Set GliderMesh(GliderMesh:Skeletal Mesh Component Budgeted Object Reference=@n3) -> (Output_Get:Skeletal Mesh Component Budgeted Object Reference)
   > n4 = Set GliderAnimInstance(GliderAnimInstance:Anim Instance Object Reference=GetAnimInstance(self=@n3)) -> (Output_Get:Anim Instance Object Reference)
   > n5 = Cast To MechanicalEngineer_Glider_AnimBP!(execute, Object=@n4) -> (then, CastFailed, AsMechanical Engineer Glider Anim BP:Mechanical Engineer Glider Anim BP Object Reference) {is=title K2Node_DynamicCast}
   > Set HideOwl(HideOwl:bool=Get HideBackpackOwl!(self:Mechanical Engineer Glider Anim BP Object Reference=@n5) -> (HideBackpackOwl:bool)) -> (Output_Get:bool)
   > Outputs()
  ]}
 EventGraph {nodes [
   n1 = Event BlueprintUpdateAnimation()
   > Set Time(Time:double=@n1.DeltaTimeX) -> (Output_Get:double)
   AnimNotify_DetachWrenchOff() {is=title K2Node_Event}
   > Set DetachFromWrenchTrans(DetachFromWrenchTrans:bool=false) -> (Output_Get:bool)
   AnimNotify_Skydive_Entered() {is=title K2Node_Event}
   > Set ResetGetFrame(ResetGetFrame:bool=true) -> (Output_Get:bool)
   AnimNotify_TheOwlHasLanded() {is=title K2Node_Event}
   > Set CanPlayLandingAnim(CanPlayLandingAnim=false)
   n2 = Get SkydivingState()
   Set SkydiveForward(SkydiveForward:double=Select(
     Option 0:double=0
     Option 1:double=2000
     Index=Break Fort Anim Input Skydiving(FortAnimInput_Skydiving=@n2) -> (
      ParaGilderRootOffset=~
      ParaGliderRootRotationOverride=~
      bUseParaGlideRootModifier=~
      bIsSkydivingFromLaunchPad=~
      bIsSkydivingFromBus=~
      bIsInVortex=~
      bIsDiving=~
      bIsParachuteOpen=~
      bIsSkydiving=~
      bIsParachuteLeaning=~
      bIsSkydiveLeaning=~
      bIsLeaning=~
      bIsSkydiveDiveMode=~
      bPlaySkydiveDrift=~
      bSkydiveDriftDelayActive=~
      bSkydiveDriftAnimAllowed=~
      bIsGliderRight=~
      bIsGliderCenter=~
      bIsGliderLeft=~
      bIsGliderForward=~
      bIsGliderBack=~
      bHasDeployedAndIsStandingUpright=~
      LocalAccelForward=~
      LocalAccelRight=~
      LocalVelocityRight=~
      SkydiveAimPitch=~
      SkydiveAimYaw=~
      DeployChuteAnimRate=~
      SkydiveDriftAnimRate=~
      SkydiveDriftAnimRateCurrent=~
      SkydiveFidgetAnimRate=~
      SkydiveFidgetAnimRateCurrent=~
      SkydiveAdditiveAlpha=~
      SkydiveDriftDelay=~
     ).bIsSkydiveDiveMode
    ) -> (ReturnValue:double)) -> (Output_Get:double)
   > n3 = Sequence!(execute) -> (then_0, then_1, then_2:Exec, then_3:Exec, then_4:Exec)
   > Check Glider!(execute:Exec) -> (then:Exec) {is=title K2Node_MacroInstance}
   Get Bone and Socket Locations!(execute:Exec=@n3.then_1) -> (then:Exec) {is=title K2Node_MacroInstance}
   n4 = Break Fort Anim Input Skydiving(FortAnimInput_Skydiving=Get SkydivingState()) -> (
    ParaGilderRootOffset=~
    ParaGliderRootRotationOverride=~
    bUseParaGlideRootModifier=~
    bIsSkydivingFromLaunchPad=~
    bIsSkydivingFromBus=~
    bIsInVortex=~
    bIsDiving=~
    bIsParachuteOpen=~
    bIsSkydiving=~
    bIsParachuteLeaning=~
    bIsSkydiveLeaning=~
    bIsLeaning=~
    bIsSkydiveDiveMode=~
    bPlaySkydiveDrift=~
    bSkydiveDriftDelayActive=~
    bSkydiveDriftAnimAllowed=~
    bIsGliderRight=~
    bIsGliderCenter=~
    bIsGliderLeft=~
    bIsGliderForward=~
    bIsGliderBack=~
    bHasDeployedAndIsStandingUpright=~
    LocalAccelForward=~
    LocalAccelRight=~
    LocalVelocityRight=~
    SkydiveAimPitch=~
    SkydiveAimYaw=~
    DeployChuteAnimRate=~
    SkydiveDriftAnimRate=~
    SkydiveDriftAnimRateCurrent=~
    SkydiveFidgetAnimRate=~
    SkydiveFidgetAnimRateCurrent=~
    SkydiveAdditiveAlpha=~
    SkydiveDriftDelay=~
   )
   Set Skydive_AimPitch(execute=@n3.then_2, Skydive_AimPitch:double=@n4.SkydiveAimPitch) -> (Output_Get:double)
   > Set Skydive_AimYaw_Target(Skydive_AimYaw_Target:double=@n4.SkydiveAimYaw) -> (Output_Get:double)
   > Set Skydive_AimYaw(Skydive_AimYaw:double=FInterpTo(
     Current=Get Skydive_AimYaw()
     Target=Get Skydive_AimYaw_Target() -> (Skydive_AimYaw_Target:double)
     DeltaTime=Get Time() -> (Time:double)
     InterpSpeed=3
    )) -> (Output_Get:double)
   > n5 = Set ClosedAlpha(ClosedAlpha:double=MapRangeClamped(Value=GetCurveValue(CurveName=Closed_New), InRangeA=0, InRangeB=1, OutRangeA=1, OutRangeB=0)) -> (Output_Get:double)
   > Set AdditiveAlpha(AdditiveAlpha:double=MapRangeClamped(Value=GetCurveValue(CurveName=AdditiveNodes), InRangeA=0, InRangeB=1, OutRangeA=1, OutRangeB=0)) -> (Output_Get:double)
   Owl Offset!(execute:Exec=@n3.then_3) -> (then:Exec) {is=title K2Node_MacroInstance}
   > Check Hide Owl Status!(execute:Exec) -> (then:Exec) {is=title K2Node_MacroInstance}
   > Match Owl Status!(execute:Exec) -> (then:Exec) {is=title K2Node_MacroInstance}
   Branch(execute=@n3.then_4, Condition=Get GliderMatch())
   > n6 = Sequence!(execute) -> (then_0, then_1)
   > Branch(Condition=Get bIsOnGround() -> (bIsOnGround:bool))
   > Do Once(Reset=@n7.then, Start Closed=true)
   > Set CanPlayLandingAnim(CanPlayLandingAnim=true)
   n7 = Branch(
    execute=@n6.then_1
    Condition=Break Fort Anim Input Skydiving(FortAnimInput_Skydiving=Get SkydivingState()) -> (
     ParaGilderRootOffset=~
     ParaGliderRootRotationOverride=~
     bUseParaGlideRootModifier=~
     bIsSkydivingFromLaunchPad=~
     bIsSkydivingFromBus=~
     bIsInVortex=~
     bIsDiving=~
     bIsParachuteOpen=~
     bIsSkydiving=~
     bIsParachuteLeaning=~
     bIsSkydiveLeaning=~
     bIsLeaning=~
     bIsSkydiveDiveMode=~
     bPlaySkydiveDrift=~
     bSkydiveDriftDelayActive=~
     bSkydiveDriftAnimAllowed=~
     bIsGliderRight=~
     bIsGliderCenter=~
     bIsGliderLeft=~
     bIsGliderForward=~
     bIsGliderBack=~
     bHasDeployedAndIsStandingUpright=~
     LocalAccelForward=~
     LocalAccelRight=~
     LocalVelocityRight=~
     SkydiveAimPitch=~
     SkydiveAimYaw=~
     DeployChuteAnimRate=~
     SkydiveDriftAnimRate=~
     SkydiveDriftAnimRateCurrent=~
     SkydiveFidgetAnimRate=~
     SkydiveFidgetAnimRateCurrent=~
     SkydiveAdditiveAlpha=~
     SkydiveDriftDelay=~
    ).bIsParachuteOpen
   )
   TryGetPawnOwner()
   PrintString!(execute, InString="To String (Float)"(InDouble=@n5), bPrintToScreen, bPrintToLog, TextColor, Duration, Key) -> (then)
   PrintString(InString="To String (Boolean)"(InBool=Get GliderMatch()))
   Print Glider Match!(execute:Exec) -> (then:Exec) {is=title K2Node_MacroInstance}
   Set Initial Attach Location!(execute:Exec) -> (then:Exec) {is=title K2Node_MacroInstance}
   Set Reset_Initial_Location(Reset_Initial_Location:bool=true) -> (Output_Get:bool)
  ]}
 GetBoneAndSocketLocations {nodes [
   Inputs()
   n1 = Get PlayerMesh() -> (PlayerMesh:Skeletal Mesh Component Object Reference)
   > Set Player_Root(Player_Root:Transform=GetSocketTransform(self=@n1, InSocketName=Root, TransformSpace=RTS_World)) -> (Output_Get:Transform)
   n2 = GetSocketTransform(self=@n1, InSocketName=attach, TransformSpace=RTS_World)
   > n3 = Set Player_AttachJoint(Player_AttachJoint:Transform=@n2) -> (Output_Get:Transform)
   Set Player_Spine_05(execute=[@n3.then, @n5.then], Player_Spine_05:Transform=GetSocketTransform(self=@n1, InSocketName=spine_05, TransformSpace=RTS_World)) -> (Output_Get:Transform)
   n4 = Get BackpackMesh() -> (BackpackMesh:Skeletal Mesh Component Object Reference)
   > Set Pack_Glider_Base_Closed(Pack_Glider_Base_Closed:Transform=GetSocketTransform(self=@n4, InSocketName=Glider_Base_Closed, TransformSpace=RTS_World)) -> (Output_Get:Transform)
   > Set Pack_Frame_Closed(Pack_Frame_Closed:Transform=GetSocketTransform(self=@n4, InSocketName=Frame_Closed, TransformSpace=RTS_World)) -> (Output_Get:Transform)
   > Set Pack_Owl_Frame(Pack_Owl_Frame:Transform=GetSocketTransform(self=@n4, InSocketName=Frame, TransformSpace=RTS_World)) -> (Output_Get:Transform)
   > Outputs()
   n5 = Set Attach_Rotation(Attach_Rotation:Rotator=BreakTransform(InTransform=@n2).Rotation) -> (Output_Get:Rotator)
  ]}
 MatchOwlStatus {nodes [
   Inputs()
   > Branch(Condition=Get GliderMatch())
   > n1 = Sequence()
   n2 = Break Fort Anim Input Skydiving(FortAnimInput_Skydiving=Get SkydivingState()) -> (
    ParaGilderRootOffset=~
    ParaGliderRootRotationOverride=~
    bUseParaGlideRootModifier=~
    bIsSkydivingFromLaunchPad=~
    bIsSkydivingFromBus=~
    bIsInVortex=~
    bIsDiving=~
    bIsParachuteOpen=~
    bIsSkydiving=~
    bIsParachuteLeaning=~
    bIsSkydiveLeaning=~
    bIsLeaning=~
    bIsSkydiveDiveMode=~
    bPlaySkydiveDrift=~
    bSkydiveDriftDelayActive=~
    bSkydiveDriftAnimAllowed=~
    bIsGliderRight=~
    bIsGliderCenter=~
    bIsGliderLeft=~
    bIsGliderForward=~
    bIsGliderBack=~
    bHasDeployedAndIsStandingUpright=~
    LocalAccelForward=~
    LocalAccelRight=~
    LocalVelocityRight=~
    SkydiveAimPitch=~
    SkydiveAimYaw=~
    DeployChuteAnimRate=~
    SkydiveDriftAnimRate=~
    SkydiveDriftAnimRateCurrent=~
    SkydiveFidgetAnimRate=~
    SkydiveFidgetAnimRateCurrent=~
    SkydiveAdditiveAlpha=~
    SkydiveDriftDelay=~
   )
   > n3 = Branch(Condition=@n2.bIsParachuteOpen)
   Do Once(execute=@n16.then_0, Reset=@n3.then)
   > Set OwlFrame_Alpha(OwlFrame_Alpha:double=1) -> (Output_Get:double)
   > Set DetachFromWrenchTrans(DetachFromWrenchTrans=true)
   n4 = Branch(execute=@n1.then_1, Condition=@n2.bIsParachuteOpen)
   > n5 = Sequence()
   > Set OwlFrame_Alpha(OwlFrame_Alpha=0)
   > Delay(Duration=0.2)
   > Set SwitchInterpSpeed(SwitchInterpSpeed:bool=true) -> (Output_Get:bool)
   Set OwlFrame_Transition_Speed(
    execute=[@n5.then_1, @n17.then]
    OwlFrame_Transition_Speed:double=Select(Option 0=0.1, Option 1=5, Index=Get SwitchInterpSpeed() -> (SwitchInterpSpeed:bool))
   ) -> (Output_Get:double)
   > SnapshotPose!(
    execute
    self:Anim Instance Object Reference=Get GliderAnimInstance() -> (GliderAnimInstance:Anim Instance Object Reference)
    Snapshot:"Pose Snapshot Structure (by ref)"=Get Glider_Pose()
   ) -> (then)
   n6 = BreakTransform(InTransform=Get Player_Root() -> (Player_Root:Transform))
   > Set PlayerRoot_Loc(PlayerRoot_Loc:Vector=@n6) -> (Output_Get:Vector)
   > Set PlayerRoot_Rot(PlayerRoot_Rot:Rotator=@n6.Rotation) -> (Output_Get:Rotator)
   > Set GliderOwl_Frame(GliderOwl_Frame:Transform=GetSocketTransform(self=Get GliderMesh() -> (GliderMesh:Skeletal Mesh Component Budgeted Object Reference), InSocketName=Frame, TransformSpace=RTS_World)) -> (Output_Get:Transform)
   n7 = Break Fort Anim Input Skydiving(FortAnimInput_Skydiving=Get SkydivingState()) -> (
    ParaGilderRootOffset=~
    ParaGliderRootRotationOverride=~
    bUseParaGlideRootModifier=~
    bIsSkydivingFromLaunchPad=~
    bIsSkydivingFromBus=~
    bIsInVortex=~
    bIsDiving=~
    bIsParachuteOpen=~
    bIsSkydiving=~
    bIsParachuteLeaning=~
    bIsSkydiveLeaning=~
    bIsLeaning=~
    bIsSkydiveDiveMode=~
    bPlaySkydiveDrift=~
    bSkydiveDriftDelayActive=~
    bSkydiveDriftAnimAllowed=~
    bIsGliderRight=~
    bIsGliderCenter=~
    bIsGliderLeft=~
    bIsGliderForward=~
    bIsGliderBack=~
    bHasDeployedAndIsStandingUpright=~
    LocalAccelForward=~
    LocalAccelRight=~
    LocalVelocityRight=~
    SkydiveAimPitch=~
    SkydiveAimYaw=~
    DeployChuteAnimRate=~
    SkydiveDriftAnimRate=~
    SkydiveDriftAnimRateCurrent=~
    SkydiveFidgetAnimRate=~
    SkydiveFidgetAnimRateCurrent=~
    SkydiveAdditiveAlpha=~
    SkydiveDriftDelay=~
   )
   > n8 = Branch(Condition=@n7.bIsParachuteOpen)
   > Set GliderOwl_Frame_Static(GliderOwl_Frame_Static:Transform=Get GliderOwl_Frame() -> (GliderOwl_Frame:Transform)) -> (Output_Get:Transform)
   > Set Player_Root_Static(Player_Root_Static:Transform=Get Player_Root()) -> (Output_Get:Transform)
   > Set OwlFrame_Alpha(OwlFrame_Alpha=FInterpTo(
     Current=Get OwlFrame_Alpha()
     Target=1
     DeltaTime=Get Time()
     InterpSpeed=Get OwlFrame_Transition_Speed() -> (OwlFrame_Transition_Speed:double)
    ))
   > n9 = Set Owl_Pose_Alpha(Owl_Pose_Alpha:double=FInterpTo(Current=Get Owl_Pose_Alpha(), Target=1, DeltaTime=Get Time(), InterpSpeed=2)) -> (Output_Get:double)
   n10 = BreakTransform(InTransform=Get GliderOwl_Frame_Static() -> (GliderOwl_Frame_Static:Transform))
   n11 = Get Player_Root_Static() -> (Player_Root_Static:Transform)
   Set GliderOwl_Frame_Loc(
    execute=[@n9.then, @n15.then]
    GliderOwl_Frame_Loc:Vector=vector - vector!(A:Vector=@n10, B:Vector=BreakTransform(InTransform=@n11)) -> (ReturnValue:Vector) {is=title K2Node_CallFunction}
   ) -> (Output_Get:Vector)
   > Set GliderOwl_Frame_Rot(GliderOwl_Frame_Rot:Rotator=@n10.Rotation) -> (Output_Get:Rotator)
   > Outputs()
   n12 = Sequence(execute=@n8.else)
   > n13 = Cast To MechanicalEngineer_Glider_AnimBP(Object=Get GliderAnimInstance())
   > n14 = Set HideBackpackOwl!(execute, HideBackpackOwl:bool=false, self:Mechanical Engineer Glider Anim BP Object Reference=@n13) -> (then, Output_Get:bool)
   > Set HideOwl(HideOwl=@n14)
   Set OwlFrame_Alpha(execute=@n12.then_1, OwlFrame_Alpha=FInterpTo(Current=Get OwlFrame_Alpha(), Target=0, DeltaTime=Get Time(), InterpSpeed=5))
   > n15 = Set Owl_Pose_Alpha(Owl_Pose_Alpha=FInterpTo(Current=Get Owl_Pose_Alpha(), Target=0, DeltaTime=Get Time(), InterpSpeed=5))
   n16 = Sequence(execute=@n4.else)
   n17 = Set SwitchInterpSpeed(execute=@n16.then_1, SwitchInterpSpeed=false)
   NOT Boolean(A=@n2.bIsParachuteOpen)
   PrintString(InString="To String (Float)"(InDouble=Get OwlFrame_Alpha()), TextColor=(R=0.026,G=0.726,B=0,A=1))
   NOT Boolean(A=@n7.bIsParachuteOpen)
   PrintString(InString=Hello, TextColor=(R=0.6635,G=0.4469,B=0.1268,A=1))
   "To String (Transform)"!(InTrans:"Transform (by ref)"=Get GliderOwl_Frame()) -> (ReturnValue:String)
   n18 = Get SkydivingState()
   vector + vector(
    A:Vector="0, 0, 0"
    B:Vector=Select(
     Option 0:Vector="0, 0,-75"
     Option 1:Vector="0, 0, 0"
     Index=Break Fort Anim Input Skydiving(FortAnimInput_Skydiving=@n18) -> (
      ParaGilderRootOffset=~
      ParaGliderRootRotationOverride=~
      bUseParaGlideRootModifier=~
      bIsSkydivingFromLaunchPad=~
      bIsSkydivingFromBus=~
      bIsInVortex=~
      bIsDiving=~
      bIsParachuteOpen=~
      bIsSkydiving=~
      bIsParachuteLeaning=~
      bIsSkydiveLeaning=~
      bIsLeaning=~
      bIsSkydiveDiveMode=~
      bPlaySkydiveDrift=~
      bSkydiveDriftDelayActive=~
      bSkydiveDriftAnimAllowed=~
      bIsGliderRight=~
      bIsGliderCenter=~
      bIsGliderLeft=~
      bIsGliderForward=~
      bIsGliderBack=~
      bHasDeployedAndIsStandingUpright=~
      LocalAccelForward=~
      LocalAccelRight=~
      LocalVelocityRight=~
      SkydiveAimPitch=~
      SkydiveAimYaw=~
      DeployChuteAnimRate=~
      SkydiveDriftAnimRate=~
      SkydiveDriftAnimRateCurrent=~
      SkydiveFidgetAnimRate=~
      SkydiveFidgetAnimRateCurrent=~
      SkydiveAdditiveAlpha=~
      SkydiveDriftDelay=~
     ).bIsParachuteOpen
    ) -> (ReturnValue:Vector)
   ) -> (ReturnValue:Vector) {is=title K2Node_CommutativeAssociativeBinaryOperator}
   Branch(Condition="Nearly Equal (Float)"(A=Get OwlFrame_Alpha(), B=1, ErrorTolerance=0.01))
   Set OwlFrame_Alpha(OwlFrame_Alpha=Get ClosedAlpha())
   Set DetachFromWrenchTrans(DetachFromWrenchTrans=false)
   Set Player_Root(Player_Root=GetSocketTransform(self=Get PlayerMesh(), InSocketName=Root, TransformSpace=RTS_World))
  ]}
 OwlOffset {nodes [
   Inputs()
   n1 = Break Fort Anim Input Skydiving(FortAnimInput_Skydiving=Get SkydivingState()) -> (
    ParaGilderRootOffset=~
    ParaGliderRootRotationOverride=~
    bUseParaGlideRootModifier=~
    bIsSkydivingFromLaunchPad=~
    bIsSkydivingFromBus=~
    bIsInVortex=~
    bIsDiving=~
    bIsParachuteOpen=~
    bIsSkydiving=~
    bIsParachuteLeaning=~
    bIsSkydiveLeaning=~
    bIsLeaning=~
    bIsSkydiveDiveMode=~
    bPlaySkydiveDrift=~
    bSkydiveDriftDelayActive=~
    bSkydiveDriftAnimAllowed=~
    bIsGliderRight=~
    bIsGliderCenter=~
    bIsGliderLeft=~
    bIsGliderForward=~
    bIsGliderBack=~
    bHasDeployedAndIsStandingUpright=~
    LocalAccelForward=~
    LocalAccelRight=~
    LocalVelocityRight=~
    SkydiveAimPitch=~
    SkydiveAimYaw=~
    DeployChuteAnimRate=~
    SkydiveDriftAnimRate=~
    SkydiveDriftAnimRateCurrent=~
    SkydiveFidgetAnimRate=~
    SkydiveFidgetAnimRateCurrent=~
    SkydiveAdditiveAlpha=~
    SkydiveDriftDelay=~
   )
   n2 = Get ClosedAlpha()
   n3 = Select(
    Option 0=Select(
     Option 0="0,165,175"
     Option 1="0, 0, 0"
     Index="Equal (Float)"!(A:double=@n2, B:double=0) -> (ReturnValue:bool) {is=title K2Node_CallFunction}
    )
    Option 1="9,26,276"
    Index=Get GliderMatch()
   )
   n4 = Get GliderMatch()
   > Set Owl_Offset_Tran(Owl_Offset_Tran:Vector=VInterpTo!(
     Current:Vector=Get Owl_Offset_Tran()
     Target:Vector=Select(Option 0="0,165,175", Option 1=@n3, Index=@n1.bIsParachuteOpen)
     DeltaTime:float=Get Time()
     InterpSpeed:float=Select(Option 0:double=5, Option 1:double=1, Index=@n4) -> (ReturnValue:double)
    ) -> (ReturnValue:Vector)) -> (Output_Get:Vector)
   > Set Owl_Offset_Rot(Owl_Offset_Rot:Rotator="0,180, 0") -> (Output_Get:Rotator)
   > Set LandingPhase(LandingPhase:bool="Equal (Float)"(A=GetCurveValue(CurveName=Landing), B=1)) -> (Output_Get:bool)
   n5 = BreakTransform(InTransform=Get Player_Root())
   > Set Attach_Rotation(Attach_Rotation=@n5.Rotation)
   n6 = "Lerp (Vector)"!(
    A:Vector=@n5
    B:Vector=BreakTransform(InTransform=Get Pack_Glider_Base_Closed() -> (Pack_Glider_Base_Closed:Transform))
    Alpha:float=GetCurveValue(CurveName=Backpack)
   ) -> (ReturnValue:Vector)
   > Set Attach_Location(Attach_Location:Vector=@n6) -> (Output_Get:Vector)
   n7 = BreakTransform(InTransform=Get Pack_Frame_Closed() -> (Pack_Frame_Closed:Transform))
   n8 = BreakTransform(InTransform=Get Pack_Owl_Frame() -> (Pack_Owl_Frame:Transform))
   > Set Frame_Location(Frame_Location:Vector="Lerp (Vector)"(A=@n8, B=@n7, Alpha=GetCurveValue(CurveName=Backpack))) -> (Output_Get:Vector)
   > Set Frame_Rot(Frame_Rot:Rotator="Lerp (Rotator)"!(A:Rotator=@n8.Rotation, B:Rotator=@n7.Rotation, Alpha:float=GetCurveValue(CurveName=Frame_Rotation), bShortestPath:bool=true) -> (ReturnValue:Rotator)) -> (Output_Get:Rotator)
   > Outputs()
   n9 = Get Player_Root()
   n10 = Get Pack_Glider_Base_Closed()
   Select(
    Option 0:Vector=@n6
    Option 1:Vector="Lerp (Vector)"(A=BreakTransform(InTransform=@n9), B=BreakTransform(InTransform=@n10), Alpha=GetCurveValue(CurveName=Backpack))
    Index=Get LandingPhase()
   ) -> (ReturnValue:Vector)
   Select(Option 0="0, 0, 0", Option 1="9,26,276", Index=Get GliderMatch())
   Get Player_AttachJoint() -> (Player_AttachJoint:Transform)
   Set Owl_Offset_Tran(Owl_Offset_Tran="2.103,70.631,146.633")
   Set Owl_Offset_Tran(Owl_Offset_Tran="0,150,200")
   Set Owl_Offset_Rot(Owl_Offset_Rot="0,180, 0")
   Set Owl_Offset_Tran(Owl_Offset_Tran="0,165,175")
   Set Owl_Offset_Tran(Owl_Offset_Tran="8.841,26.226,276.45")
  ]}
 PrintGliderMatch {nodes [
   Inputs()
   n1 = Get GliderMatch()
   > PrintString(InString=Append(A="GliderMatch: ", B="To String (Boolean)"(InBool=@n1)), TextColor=(R=0.8094,G=0.4244,B=0.0629,A=1))
   n2 = Get bIsOnGround()
   > PrintString(InString=Append(A="OnGround: ", B="To String (Boolean)"(InBool=@n2)))
   > Outputs()
   n3 = Get GliderPlayerAnimSet()
   PrintString(
    InString=Append(A="AnimSet: ", B="To String (SoftObjectReference)"!(SoftObjectReference:"Soft Object Reference (by ref)"=@n3) -> (ReturnValue:String))
    TextColor=(R=0.2188,G=0.3254,B=0.8094,A=1)
   )
  ]}
 SetInitialAttachLocation {nodes [
   Inputs()
   > n1 = Sequence()
   > Do Once(Reset=@n2.then)
   > Set Attach_Location(Attach_Location=BreakTransform(InTransform=Get Player_AttachJoint()))
   > Set Reset_Initial_Location(Reset_Initial_Location=false)
   > Outputs()
   n2 = Branch(execute=@n1.then_1, Condition=Get Reset_Initial_Location() -> (Reset_Initial_Location:bool))
  ]}
]
variables [
 AdditiveAlpha {type=number}
 Attach_Location {type=Vector}
 Attach_Rotation {type=Rotator}
 BackpackMesh {type=SkeletalMeshComponent}
 BlockAnimRule {type=boolean}
 CanPlayLandingAnim {type=boolean}
 CharMovement {type=CharacterMovementComponent, default=null}
 ClosedAlpha {type=number}
 DetachFromWrenchTrans {type=boolean}
 Frame_Location {type=Vector}
 Frame_Rot {type=Rotator}
 Glider_Pose {type=PoseSnapshot, default {bIsValid=false, skeletalMeshName=None, snapshotName=None, boneNames [], localTransforms []}}
 GliderAnimInstance {type=AnimInstance, default=null}
 GliderMatch {type=boolean}
 GliderMesh {type=SkeletalMeshComponentBudgeted}
 GliderOwl_Frame {type=Transform, default {location=<0 0 0>, rotation=<rot 0 0 0>, scale=<1 1 1>}}
 GliderOwl_Frame_Loc {type=Vector}
 GliderOwl_Frame_Rot {type=Rotator}
 GliderOwl_Frame_Static {type=Transform, default {location=<0 0 0>, rotation=<rot 0 0 0>, scale=<1 1 1>}}
 GliderPlayerAnimSet {type=FortPlayerGliderAnimSet, default=null}
 GliderSkelMesh {
  type=SkeletalMesh
  default=/BRCosmetics/Gadgets/Assets/VinderTech_GliderChute/Glider_Mechanical_Engineer_Female/Meshes/Mechanical_Engineer_Female_Glider
 }
 HangGliderAnimSet {type=FortPlayerGliderAnimSet, default=/Game/Animation/Game/MainPlayer/Skydive/ParaGlide/HangGlider/HangGliderPlayerAnimSet}
 HideOwl {type=boolean}
 LandingPhase {type=boolean}
 Owl_Offset_Rot {type=Rotator, default=<rot 0 180 0>}
 Owl_Offset_Tran {type=Vector, default=<0 150 200>}
 Owl_Pose_Alpha {type=number}
 OwlFrame_Alpha {type=number}
 OwlFrame_Transition_Speed {type=number}
 Pack_Frame_Closed {type=Transform, default {location=<0 0 0>, rotation=<rot 0 0 0>, scale=<1 1 1>}}
 Pack_Glider_Base_Closed {type=Transform, default {location=<0 0 0>, rotation=<rot 0 0 0>, scale=<1 1 1>}}
 Pack_Owl_Frame {type=Transform, default {location=<0 0 0>, rotation=<rot 0 0 0>, scale=<1 1 1>}}
 Player_AttachJoint {type=Transform, default {location=<0 0 0>, rotation=<rot 0 0 0>, scale=<1 1 1>}}
 Player_Root {type=Transform, default {location=<0 0 0>, rotation=<rot 0 0 0>, scale=<1 1 1>}}
 Player_Root_Static {type=Transform, default {location=<0 0 0>, rotation=<rot 0 0 0>, scale=<1 1 1>}}
 Player_Spine_05 {type=Transform, default {location=<0 0 0>, rotation=<rot 0 0 0>, scale=<1 1 1>}}
 PlayerMesh {type=SkeletalMeshComponent}
 PlayerRoot_Loc {type=Vector}
 PlayerRoot_Rot {type=Rotator}
 Reset_Initial_Location {type=boolean}
 ResetGetFrame {type=boolean}
 Skydive_AimPitch {type=number}
 Skydive_AimYaw {type=number}
 Skydive_AimYaw_Target {type=number}
 SkydiveForward {type=number}
 SwitchInterpSpeed {type=boolean}
 Time {type=number}
 UmbrellaAnimSet {type=FortPlayerGliderAnimSet, default=/Game/Animation/Game/MainPlayer/Skydive/ParaGlide/Umbrella/UmbrellaPlayerAnimSet}
]
